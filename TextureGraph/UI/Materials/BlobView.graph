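# Unlit UI-domain material for the TextureGraph plugin, named for displaying a blob texture in a blob view widget.
# A texture object parameter BlobTex, texture coordinates and a BlobTexIsValid flag parameter feed a single Custom HLSL node.
# The node's logic lives in the included BlobViewWidget.ush, with only a fallback return of zero written in the material itself.
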Material /TextureGraph/UI/Materials/BlobView
graph {
 properties {materialDomain=MD_UI, pixelDepthOffsetMode=PDOM_Legacy, shadingModel=MSM_Unlit, usedShadingModels=Unlit}
 nodes [
  MaterialExpressionTextureObjectParameter_0 = TextureObjectParameter() {parameterName=BlobTex, samplerType=SAMPLERTYPE_LinearColor, texture=/Engine/EngineMaterials/DefaultCalibrationColor}
  MaterialExpressionTextureCoordinate_0 = TextureCoordinate()
  MaterialExpressionScalarParameter_0 = ScalarParameter() {desc=Flag to notify that the BlobTex is not valid, parameterName=BlobTexIsValid}
  Custom!(
   blobTex=@MaterialExpressionTextureObjectParameter_0
   texcoord=@MaterialExpressionTextureCoordinate_0
   blobTexIsValid=@MaterialExpressionScalarParameter_0
  ) {
   code="##include \"/Plugin/TextureGraph/BlobViewWidget.ush\"\r\nreturn 0;      "
   inputs [
    {inputName=blobTex, input {expression=&self:MaterialExpressionTextureObjectParameter_0}}
    {inputName=texcoord, input {expression=&self:MaterialExpressionTextureCoordinate_0}}
    {inputName=blobTexIsValid, input {expression=&self:MaterialExpressionScalarParameter_0}}
   ]
  }
 ]
}
