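# CheckerPattern is an unlit UI-domain material in TextureGraph that draws a tiled checkerboard.
# It calls the engine's ArtTools CheckerPattern material function, feeding it texture coordinates as UVs and a constant two-component vector for uniform tiling.
# It exposes no parameters, so the pattern density is fixed in the graph.
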Material /TextureGraph/UI/Materials/CheckerPattern
graph {
 properties {materialDomain=MD_UI, pixelDepthOffsetMode=PDOM_Legacy, shadingModel=MSM_Unlit, usedShadingModels=Unlit}
 nodes [MaterialFunctionCall!(UVs=TextureCoordinate(), Tiling=Constant2Vector() {g=10, r=10}) {materialFunction=/Engine/ArtTools/RenderToTexture/MaterialFunctions/CheckerPattern}]
}
