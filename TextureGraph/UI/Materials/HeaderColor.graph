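# HeaderColor is a UI-domain material in TextureGraph, named as a header color for the plugin's interface.
# Its graph holds a single constant vector of very dark neutral gray, with no textures, parameters or math.
# The graph lists no link from that constant to a material output.
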Material /TextureGraph/UI/Materials/HeaderColor
graph {
 properties {materialDomain=MD_UI, pixelDepthOffsetMode=PDOM_Legacy, usedShadingModels=Default Lit}
 nodes [Constant3Vector() {constant=<rgba 0.0296 0.0296 0.0296 0>}]
}
