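# MaskComponent is a UI-domain material in TextureGraph, with a procedural checker pattern and a MaskTexture parameter, apparently for mask display.
# It rounds U and V, then adds their clamped two-way differences as an XOR and dims the result to 0.2.
# The V branch sets only U tiling, so that tiling has no effect.
# The graph shows no link from MaskTexture or the checker to a material output.
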Material /TextureGraph/UI/Materials/MaskComponent
graph {
 properties {materialDomain=MD_UI, pixelDepthOffsetMode=PDOM_Legacy, usedShadingModels=Default Lit}
 nodes [
  n1 = Round(None=ComponentMask(None=TextureCoordinate()) {r=true})
  n2 = Round(None=ComponentMask(None=TextureCoordinate() {uTiling=4}) {g=true})
  n3 = Subtract(A=@n1, B=@n2)
  n4 = Subtract(A=@n2, B=@n1)
  Multiply(A=Add(A=Clamp(None=@n3), B=Clamp(None=@n4))) {constB=0.2}
  TextureSampleParameter2D() {parameterName=MaskTexture}
 ]
}
