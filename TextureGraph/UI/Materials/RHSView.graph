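# RHSView is an unlit UI-domain material in TextureGraph that samples a replaceable texture, likely for a view panel.
# A texture object parameter named BlobTex feeds a single texture sample at default UVs, with an engine calibration color texture as placeholder.
# Both use the shared clamp world-group sampler.
# The graph shows no connection from the sample to a material output.
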Material /TextureGraph/UI/Materials/RHSView
graph {
 properties {materialDomain=MD_UI, pixelDepthOffsetMode=PDOM_Legacy, shadingModel=MSM_Unlit, usedShadingModels=Unlit}
 nodes [TextureSample(
   UVs=TextureCoordinate()
   Tex=TextureObjectParameter() {parameterName=BlobTex, samplerSource=SSM_Clamp_WorldGroupSettings, texture=/Engine/EngineMaterials/DefaultCalibrationColor}
  ) {samplerSource=SSM_Clamp_WorldGroupSettings}]
}
