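# RoundedBorder is a translucent UI material in TextureGraph that pairs solid white with a circular mask, presumably for rounded interface borders.
# A RadialGradientExponential function builds the circle over the UVs, centered in the element with a radius reaching its edges.
# The Density scalar parameter controls falloff, and its high default gives the circle a near-hard edge.
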Material /TextureGraph/UI/Materials/RoundedBorder
graph {
 properties {blendMode=BLEND_Translucent, materialDomain=MD_UI, pixelDepthOffsetMode=PDOM_Legacy, usedShadingModels=Default Lit}
 nodes [
  Constant3Vector() {constant=<rgba 1 1 1 0>}
  MaterialFunctionCall(
   UVs=TextureCoordinate()
   CenterPosition=AppendVector(A=Constant() {r=0.5}, B=Constant() {r=0.5})
   Radius=Constant() {r=0.5}
   Density=ScalarParameter() {defaultValue=100, parameterName=Density}
  ) {materialFunction=&RadialGradientExponential}
 ]
}
